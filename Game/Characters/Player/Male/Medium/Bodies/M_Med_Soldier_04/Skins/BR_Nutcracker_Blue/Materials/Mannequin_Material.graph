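# Default Lit skeletal mesh material named Mannequin_Material, stored with the BR_Nutcracker_Blue skin of a male soldier character.
# A color parameter times the AO texture drives base color, the AO texture alone drives specular and occlusion, and parameters supply metallic, roughness and normal.
# These material attributes pass through MF_TechArtOverrides.
# A Fresnel cloth fuzz branch with a tiled mask is present but never feeds those attributes.
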Material /Game/Characters/Player/Male/Medium/Bodies/M_Med_Soldier_04/Skins/BR_Nutcracker_Blue/Materials/Mannequin_Material
graph {
 properties {
  bUseMaterialAttributes=true
  bUsedWithMeshDeformer=true
  bUsedWithMorphTargets=true
  bUsedWithSkeletalMesh=true
  pixelDepthOffsetMode=PDOM_Legacy
  usedShadingModels=Default Lit
 }
 nodes [
  n1 = OneMinus(None=DotProduct(A=PixelNormalWS(), B=CameraVectorWS()))
  n2 = ScalarParameter() {defaultValue=4, group=Cloth, parameterName=Fuzz Exponent}
  n3 = TextureSampleParameter2D(UVs=Multiply(A=TextureCoordinate(), B=ScalarParameter() {defaultValue=4, group=Cloth, parameterName=Fuzz Tiling})) {
   group=Cloth
   parameterName=ClothFuzz Texture
   samplerSource=SSM_Wrap_WorldGroupSettings
   texture=/Game/Global/Textures/Default/Masks_and_Gradients/T_Fuzz_MASK
  }
  n4 = Multiply(A=Clamp(None=Power(Base=@n1, Exp=@n2)), B=@n3.R)
  n5 = ScalarParameter() {group=Cloth, parameterName=Fuzz Fresnel Blend}
  Add(A=@n4, B=@n5)
  n6 = TextureSampleParameter2D() {parameterName=AO, texture=/Game/UI/Frontend/Lobby/SocialImport/Textures/FB-f-Logo__white_29}
  n7 = Multiply(A=VectorParameter() {parameterName=Color, defaultValue=<rgba 0.1864 0.1846 0.217 0>}.RGB, B=@n6.RGB)
  n8 = Multiply(A=Constant() {r=1}, B=@n6.RGB)
  n9 = SetMaterialAttributes!(
   MaterialAttributes
   Base Color=NamedRerouteDeclaration(None=@n7) {name=BaseColor, nodeColor=<rgba 0.6172 1 0 1>}
   Metallic=NamedRerouteDeclaration(None=ScalarParameter() {parameterName=Metalness}) {name=Metallic, nodeColor=<rgba 1 0 0.4687 1>}
   Specular=NamedRerouteDeclaration(None=@n8) {name=Specular, nodeColor=<rgba 0 1 0.4141 1>}
   Roughness=NamedRerouteDeclaration(None=ScalarParameter() {defaultValue=0.9063, parameterName=Rough}) {name=Roughness, nodeColor=<rgba 0.7109 0 1 1>}
   Normal=NamedRerouteDeclaration(None=TextureSampleParameter2D() {parameterName=Normal, samplerType=SAMPLERTYPE_Normal, texture=&FlatNormalBlue}.RGB) {name=Nomral, nodeColor=<rgba 0.8516 1 0 1>}
   Ambient Occlusion=NamedRerouteDeclaration(None=@n8) {name=AmbOcc, nodeColor=<rgba 0.875 0 1 1>}
  ) {attributeSetTypes [
    #1
    #2
    #3
    #4
    #5
    #6
   ]}
  MaterialFunctionCall(In=@n9) {materialFunction=&MF_TechArtOverrides}
  Multiply(A=@n7, B=Saturate(None=Multiply(A=@n4, B=@n5)))
  ScalarParameter() {defaultValue=0.5, parameterName=Value}
 ]
}
